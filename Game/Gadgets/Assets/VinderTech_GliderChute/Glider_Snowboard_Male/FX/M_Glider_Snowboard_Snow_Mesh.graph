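# Masked subsurface material for snow Niagara mesh and sprite particles in the male snowboard glider's effects.
# Base color adds a Fresnel edge glow, a time-of-day function blends four subsurface colors, and particle alpha is optional via a parameter.
# A tight highlight from the normal map dotted with the sun/moon direction, tinted by particle color, feeds emissive adjustments.
# An optional brightness boost varies by shading path, stronger on forward and mobile.
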Material /Game/Gadgets/Assets/VinderTech_GliderChute/Glider_Snowboard_Male/FX/M_Glider_Snowboard_Snow_Mesh
graph {
 properties {
  bUsedWithNiagaraMeshParticles=true
  bUsedWithNiagaraSprites=true
  blendMode=BLEND_Masked
  pixelDepthOffsetMode=PDOM_Legacy
  shadingModel=MSM_Subsurface
  usedShadingModels=Subsurface
 }
 nodes [
  Add(
   A=VectorParameter() {parameterName=BaseColor, defaultValue=<rgba 0.4167 0.4167 0.4167 0>}.RGB
   B=Multiply(A=VectorParameter() {parameterName=EdgeGlow, defaultValue=<rgba 0.5277 0.6176 1 0>}.RGB, B=Fresnel() {baseReflectFraction=0, exponent=2})
  )
  n1 = ParticleColor()
  LinearInterpolate(A=Constant() {r=1}, B=@n1.A, Alpha=ScalarParameter() {parameterName=Use_Particle_Color_Alpha})
  MaterialFunctionCall(
   Color_Morning=VectorParameter() {parameterName=Subsurface Color morning, defaultValue=<rgba 0.2901 0.4801 0.599 0>}.RGB
   Color_Day=VectorParameter() {group=color time of day, parameterName=Subsurface Color, defaultValue=<rgba 0.4415 0.7305 0.9115 0>}.RGB
   Color_Evening=VectorParameter() {parameterName=Subsurface Color evening, defaultValue=<rgba 0.4415 0.7305 0.9115 0>}.RGB
   Color_Night=VectorParameter() {parameterName=Subsurface Color night, defaultValue=<rgba 0.9115 0.9115 0.9115 0>}.RGB
  ) {materialFunction=&Color_Time_of_Day}
  n2 = Constant() {r=1}
  n3 = Constant() {r=5}
  n4 = Constant() {r=13}
  n5 = ShadingPathSwitch(Default=@n3, Deferred=@n3, Forward=@n4, Mobile=@n4)
  n6 = ScalarParameter() {parameterName=Use medium and low brightness adjustments in frontend}
  n7 = CollectionParameter() {
   collection=&FortniteMaterialParameters
   parameterName=SunAndMoonModelDirectionalVector
  }
  n8 = TextureSampleParameter2D(UVs=TextureCoordinate() {vTiling=0.5}) {parameterName=Base Normal, samplerType=SAMPLERTYPE_Normal, texture=&T_Weap_Distort_N_02}
  n9 = Power(Base=DotProduct(A=ComponentMask(None=@n7) {b=true, g=true, r=true}, B=@n8.RGB)) {constExponent=50}
  n10 = Multiply(A=Multiply(A=LinearInterpolate(Alpha=@n9) {constA=0.5}), B=@n1.RGB)
  MaterialFunctionCall(Emissive=Multiply(A=LinearInterpolate(A=@n2, B=@n5, Alpha=@n6), B=Multiply(A=@n10) {constB=0.1})) {materialFunction=&MF_FX_EmissiveAdjustments}
  ScalarParameter() {parameterName=Metallic}
  ScalarParameter() {defaultValue=0.5, parameterName=Specular}
  ScalarParameter() {defaultValue=0.5, parameterName=Roughness}
  TextureSampleParameter2D(UVs=TextureCoordinate()) {parameterName=Base Normal, samplerType=SAMPLERTYPE_Normal, texture=&T_Weap_Distort_N_02}
  VectorParameter() {group=color time of day, parameterName=Subsurface Color, defaultValue=<rgba 0.4415 0.7305 0.9115 0>}
 ]
}
