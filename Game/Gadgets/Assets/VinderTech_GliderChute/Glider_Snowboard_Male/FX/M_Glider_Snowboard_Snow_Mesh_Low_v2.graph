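# Lit material for Niagara mesh particles, by its name a low-detail snow effect for the male snowboard glider.
# A tiled distortion normal map is dotted with the global sunlight vector, then smoothstepped into an adjustable mask that fakes light variation.
# That mask shades particle color into a Base Color reroute, which is also multiplied by a time-of-day tint blended from morning, day, evening and night colors.
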Material /Game/Gadgets/Assets/VinderTech_GliderChute/Glider_Snowboard_Male/FX/M_Glider_Snowboard_Snow_Mesh_Low_v2
graph {
 properties {bUsedWithNiagaraMeshParticles=true, pixelDepthOffsetMode=PDOM_Legacy, usedShadingModels=Default Lit}
 nodes [
  Multiply(
   A=NamedRerouteUsage()
   B=MaterialFunctionCall(
    Color_Morning=VectorParameter() {parameterName=Subsurface Color morning, defaultValue=<rgba 0.2901 0.4801 0.599 0>}.RGB
    Color_Day=VectorParameter() {group=color time of day, parameterName=Subsurface Color, defaultValue=<rgba 0.4397 0.7305 0.9131 0>}.RGB
    Color_Evening=VectorParameter() {parameterName=Subsurface Color evening, defaultValue=<rgba 0.4415 0.7305 0.9115 0>}.RGB
    Color_Night=VectorParameter() {parameterName=Subsurface Color night, defaultValue=<rgba 0.9115 0.9115 0.9115 0>}.RGB
   ) {materialFunction=&Color_Time_of_Day}.Result
  )
  n1 = ScalarParameter() {defaultValue=0.35, parameterName=Texture Amount}
  n2 = ScalarParameter() {defaultValue=0.2, parameterName=Texture Softness}
  n3 = CollectionParameter() {
   collection=&FortniteMaterialParameters
   parameterName=SunlightDirectionalVector
  }
  n4 = TextureSampleParameter2D(UVs=TextureCoordinate() {vTiling=0.5}) {parameterName=Base Normal, samplerType=SAMPLERTYPE_Normal, texture=&T_Weap_Distort_N_02}
  n5 = DotProduct(A=ComponentMask(None=@n3) {b=true, g=true, r=true}, B=@n4.RGB)
  n6 = Saturate(None=SmoothStep(Min=@n1, Max=Add(A=@n1, B=@n2), Value=ConstantBiasScale(None=@n5)))
  NamedRerouteDeclaration(None=Multiply(A=ParticleColor().RGB, B=LinearInterpolate(Alpha=@n6) {constA=0.5})) {name=Base Color, nodeColor=<rgba 0 1 0.9531 1>}
  ParticleColor()
  ScalarParameter() {defaultValue=0.5, parameterName=Roughness}
  ScalarParameter() {parameterName=Metalic}
  ScalarParameter() {defaultValue=0.5, parameterName=Specular}
  TextureSampleParameter2D(UVs=TextureCoordinate()) {parameterName=Base Normal, samplerType=SAMPLERTYPE_Normal, texture=&T_Weap_Distort_N_02}
 ]
}
